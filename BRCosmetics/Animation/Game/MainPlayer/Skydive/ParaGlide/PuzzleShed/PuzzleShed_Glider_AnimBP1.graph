AnimBlueprint /BRCosmetics/Animation/Game/MainPlayer/Skydive/ParaGlide/PuzzleShed/PuzzleShed_Glider_AnimBP1
target_skeleton=/BRCosmetics/Gadgets/Assets/VinderTech_GliderChute/Glider_PuzzleShed/Meshes/PuzzleShed_Glider_Skeleton
graphs [
 AnimGraph {nodes [
   n1 = "Transform (Modify) Bone - Bone: body_back"!(
    ComponentPose=Local To Component(LocalPose=Use cached pose 'IK Legs'())
    bAlphaBoolEnabled:bool=Get IsInFrontEnd() -> (IsInFrontEnd:bool)
    Translation="0,0,0"
    Rotation="0,0,0"
    Scale="0.5,1,1"
   ) -> (Pose) {
    alphaInputType=Bool
    rotationMode=BMM_Additive
    rotationSpace=BCS_BoneSpace
    scaleMode=BMM_Additive
    scaleSpace=BCS_BoneSpace
    translationMode=BMM_Additive
    translationSpace=BCS_BoneSpace
    alphaBoolBlend {blendOption=Linear}
    boneToModify {boneName=body_back}
    scale {x=0.5}
   }
   n2 = "Transform (Modify) Bone - Bone: body_front"!(ComponentPose=@n1, bAlphaBoolEnabled:bool=Get IsInFrontEnd(), Translation="0,0,10", Rotation="-50,0,0", Scale="1,1,1") -> (Pose) {
    alphaInputType=Bool
    rotationMode=BMM_Additive
    rotationSpace=BCS_BoneSpace
    translationMode=BMM_Additive
    translationSpace=BCS_BoneSpace
    alphaBoolBlend {blendOption=Linear}
    boneToModify {boneName=body_front}
    rotation {pitch=-50}
    translation {z=10}
   }
   n3 = Copy Bone - Source Bone: paw_front_r_05 - Target Bone: VB Root_paw_front_r_05!(
    ComponentPose:Component Space Pose Link Structure=@n2
    Alpha:float=1
    bCopyTranslation:bool=true
    bCopyRotation:bool=true
    bCopyScale:bool=true
   ) -> (Pose:Component Space Pose Link Structure) {
    is=title AnimGraphNode_CopyBone
    controlSpace=BCS_BoneSpace
    alphaBoolBlend {blendOption=Linear}
    sourceBone {boneName=paw_front_r_05}
    targetBone {boneName=VB Root_paw_front_r_05}
   }
   n4 = Copy Bone - Source Bone: paw_front_l_05 - Target Bone: VB Root_paw_front_l_05!(
    ComponentPose:Component Space Pose Link Structure=@n3
    Alpha:float=1
    bCopyTranslation:bool=true
    bCopyRotation:bool=true
    bCopyScale:bool=true
   ) -> (Pose:Component Space Pose Link Structure) {
    is=title AnimGraphNode_CopyBone
    controlSpace=BCS_BoneSpace
    alphaBoolBlend {blendOption=Linear}
    sourceBone {boneName=paw_front_l_05}
    targetBone {boneName=VB Root_paw_front_l_05}
   }
   n5 = Copy Bone - Source Bone: paw_back_r_05 - Target Bone: VB Root_paw_back_r_05!(
    ComponentPose:Component Space Pose Link Structure=@n4
    Alpha:float=1
    bCopyTranslation:bool=true
    bCopyRotation:bool=true
    bCopyScale:bool=true
   ) -> (Pose:Component Space Pose Link Structure) {
    is=title AnimGraphNode_CopyBone
    controlSpace=BCS_BoneSpace
    alphaBoolBlend {blendOption=Linear}
    sourceBone {boneName=paw_back_r_05}
    targetBone {boneName=VB Root_paw_back_r_05}
   }
   n6 = Copy Bone - Source Bone: paw_back_l_05 - Target Bone: VB Root_paw_back_l_05!(
    ComponentPose:Component Space Pose Link Structure=@n5
    Alpha:float=1
    bCopyTranslation:bool=true
    bCopyRotation:bool=true
    bCopyScale:bool=true
   ) -> (Pose:Component Space Pose Link Structure) {
    is=title AnimGraphNode_CopyBone
    controlSpace=BCS_BoneSpace
    alphaBoolBlend {blendOption=Linear}
    sourceBone {boneName=paw_back_l_05}
    targetBone {boneName=VB Root_paw_back_l_05}
   }
   n7 = Leg IK!(ComponentPose:Component Space Pose Link Structure=@n6, bAlphaBoolEnabled:bool=NOT Boolean(A=Get IsInFrontEnd())) -> (Pose:Component Space Pose Link Structure) {
    is=title AnimGraphNode_LegIK
    alphaInputType=Bool
    maxIterations=50
    alphaBoolBlend {blendOption=Linear}
    legsDefinition [
     {
      bEnableKneeTwistCorrection=true
      bEnableRotationLimit=false
      footBoneForwardAxis=Y
      hingeRotationAxis=None
      minRotationAngle=15
      numBonesInLimb=4
      twistOffsetCurveName=None
      fKFootBone {boneName=paw_front_l_05}
      iKFootBone {boneName=VB Root_paw_front_l_05}
     }
     {
      bEnableKneeTwistCorrection=true
      bEnableRotationLimit=false
      footBoneForwardAxis=Y
      hingeRotationAxis=None
      minRotationAngle=15
      numBonesInLimb=4
      twistOffsetCurveName=None
      fKFootBone {boneName=paw_front_r_05}
      iKFootBone {boneName=VB Root_paw_front_r_05}
     }
     {
      bEnableKneeTwistCorrection=true
      bEnableRotationLimit=false
      footBoneForwardAxis=Y
      hingeRotationAxis=None
      minRotationAngle=15
      numBonesInLimb=4
      twistOffsetCurveName=None
      fKFootBone {boneName=paw_back_r_05}
      iKFootBone {boneName=VB Root_paw_back_r_05}
     }
     {
      bEnableKneeTwistCorrection=true
      bEnableRotationLimit=false
      footBoneForwardAxis=Y
      hingeRotationAxis=None
      minRotationAngle=15
      numBonesInLimb=4
      twistOffsetCurveName=None
      fKFootBone {boneName=paw_back_l_05}
      iKFootBone {boneName=VB Root_paw_back_l_05}
     }
    ]
   }
   n8 = Copy Bone - Source Bone: VB Glider_Base_Glider_Base - Target Bone: Glider_Base!(
    ComponentPose:Component Space Pose Link Structure=@n7
    Alpha:float=1
    bCopyTranslation:bool=true
    bCopyRotation:bool=true
    bCopyScale:bool=true
   ) -> (Pose:Component Space Pose Link Structure) {
    is=title AnimGraphNode_CopyBone
    controlSpace=BCS_BoneSpace
    alphaBoolBlend {blendOption=Linear}
    sourceBone {boneName=VB Glider_Base_Glider_Base}
    targetBone {boneName=Glider_Base}
   }
   Output Pose(Result=Component To Local(ComponentPose=@n8)) {layerGroup=None}
   AnimGraphNode_StateMachine_0 = Locomotion() {is=title AnimGraphNode_StateMachine}
   AnimGraphNode_StateMachine_2 = Blink() {is=title AnimGraphNode_StateMachine}
   AnimGraphNode_StateMachine_4 = Turn Transition() {is=title AnimGraphNode_StateMachine}
   n9 = Apply Additive!(
    Base=Apply Additive!(Base=@AnimGraphNode_StateMachine_0, Additive=@AnimGraphNode_StateMachine_2, Alpha:float) -> (Pose) {alphaBoolBlend {blendOption=Linear}}
    Additive=@AnimGraphNode_StateMachine_4
    AlphaCurveName:Name=ParaGlide_Lean
   ) -> (Pose) {alphaInputType=Curve, alphaBoolBlend {blendOption=Linear}, alphaScaleBiasClamp {bias=1, scale=-1}}
   AnimGraphNode_StateMachine_6 = Lean Transition() {is=title AnimGraphNode_StateMachine}
   AnimGraphNode_StateMachine_3 = Jitter() {is=title AnimGraphNode_StateMachine}
   Save cached pose 'IK Legs'(Pose=Apply Additive!(
     Base=Apply Additive(Base=@n9, Additive=@AnimGraphNode_StateMachine_6, AlphaCurveName=ParaGlide_Lean) {alphaInputType=Curve, alphaBoolBlend {blendOption=Linear}, alphaScaleBiasClamp {bias=1, scale=-1}}
     Additive=@AnimGraphNode_StateMachine_3
     AlphaCurveName:Name=ParaGlide_Jitter
    ) -> (Pose) {alphaInputType=Curve, alphaBoolBlend {blendOption=Linear}, alphaScaleBiasClamp {bias=1, scale=-1}}) {cacheName=IK Legs}
   AnimGraphNode_StateMachine_1 = Locomotion_GliderItem() {is=title AnimGraphNode_StateMachine}
   AnimGraphNode_StateMachine_5 = Old_Turn() {is=title AnimGraphNode_StateMachine}
   AnimGraphNode_StateMachine_8 = Lean Transition_1() {is=title AnimGraphNode_StateMachine}
  ]}
 Locomotion {
  entry=Closed
  path=AnimGraph.AnimGraphNode_StateMachine_0.Locomotion
  ids=id
  nodes [
   AnimStateNode_0 = Idle_Open() {is=title AnimStateNode}
   AnimStateNode_2 = Closed() {is=title AnimStateNode, stateEntered {guid=#1, notifyName=ClosedEntered}}
   AnimStateNode_3 = Open_Into() {is=title AnimStateNode, stateEntered {guid=#2, notifyName=DeployEntered}}
   AnimStateTransitionNode_0 = AnimStateTransitionNode() {bAutomaticRuleBasedOnSequencePlayerInState=true, blendMode=Linear, crossfadeDuration=0}
   AnimStateTransitionNode_1 = AnimStateTransitionNode() {crossfadeDuration=0.1}
   AnimStateTransitionNode_2 = AnimStateTransitionNode() {crossfadeDuration=0}
   AnimStateTransitionNode_3 = AnimStateTransitionNode() {crossfadeDuration=0, priorityOrder=2}
  ]
  links [
   {
    dst=Open_Into
    rule=AnimGraph.AnimGraphNode_StateMachine_0.Locomotion.AnimStateTransitionNode_1.Transition
    src=Closed
    via=AnimStateTransitionNode_1
   }
   {
    dst=Closed
    rule=AnimGraph.AnimGraphNode_StateMachine_0.Locomotion.AnimStateTransitionNode_2.Transition
    src=Idle_Open
    via=AnimStateTransitionNode_2
   }
   {
    dst=Closed
    rule=AnimGraph.AnimGraphNode_StateMachine_0.Locomotion.AnimStateTransitionNode_3.Transition
    src=Open_Into
    via=AnimStateTransitionNode_3
   }
   {
    dst=Idle_Open
    rule=AnimGraph.AnimGraphNode_StateMachine_0.Locomotion.AnimStateTransitionNode_0.Transition
    src=Open_Into
    via=AnimStateTransitionNode_0
   }
  ]
 }
 Idle_Open {
  path=AnimGraph.AnimGraphNode_StateMachine_0.Locomotion.AnimStateNode_0.Idle_Open
  nodes [
   n1 = Break Fort Anim Input Skydiving!(FortAnimInput_Skydiving:"Fort Anim Input Skydiving Structure (by ref)"=Get SkydivingState() -> (SkydivingState:Fort Anim Input Skydiving Structure)) -> (
    ParaGilderRootOffset:Vector=~
    ParaGliderRootRotationOverride:Rotator=~
    bUseParaGlideRootModifier:bool=~
    bIsSkydivingFromLaunchPad:bool=~
    bIsSkydivingFromBus:bool=~
    bIsInVortex:bool=~
    bIsDiving:bool=~
    bIsParachuteOpen:bool=~
    bIsSkydiving:bool=~
    bIsParachuteLeaning:bool=~
    bIsSkydiveLeaning:bool=~
    bIsLeaning:bool=~
    bIsSkydiveDiveMode:bool=~
    bPlaySkydiveDrift:bool=~
    bSkydiveDriftDelayActive:bool=~
    bSkydiveDriftAnimAllowed:bool=~
    bIsGliderRight:bool=~
    bIsGliderCenter:bool=~
    bIsGliderLeft:bool=~
    bIsGliderForward:bool=~
    bIsGliderBack:bool=~
    bHasDeployedAndIsStandingUpright:bool=~
    LocalAccelForward:float=~
    LocalAccelRight:float=~
    LocalVelocityRight:float=~
    SkydiveAimPitch:float=~
    SkydiveAimYaw:float=~
    DeployChuteAnimRate:float=~
    SkydiveDriftAnimRate:float=~
    SkydiveDriftAnimRateCurrent:float=~
    SkydiveFidgetAnimRate:float=~
    SkydiveFidgetAnimRateCurrent:float=~
    SkydiveAdditiveAlpha:float=~
    SkydiveDriftDelay:float=~
    SkydiveDriftAnim:int
    SkydiveDriftAnimMax:int
    LaunchpadAnim:int
    LocalAccelDir:ESkydivingDirection Enum
    DirectionLast:ESkydivingDirection Enum
    PoseAdditive:Anim Sequence Object Reference
   ) {is=title K2Node_BreakStruct}
   n2 = Get OwnerGlider() -> (OwnerGlider:Fort Player Parachute Object Reference)
   Output Animation Pose(Result=Blend Poses by bool(
     BlendPose_0=Sequence Player 'PuzzleShed_Glide_FrontEnd_GLIDER'() {sequence=&/PuzzleShed_Glide_FrontEnd_GLIDER}
     BlendPose_1=Blendspace Player 'BS_PuzzleShed_GLIDER'(X=@n1.SkydiveAimYaw, Y=@n1.SkydiveAimPitch) {blendSpace=&/BS_PuzzleShed_GLIDER}
     bActiveValue=Get IsCosmeticPreview!(self:Fort Player Parachute Object Reference=@n2) -> (IsCosmeticPreview:bool)
    ) {blendType=Linear})
  ]
 }
 Closed {
  path=AnimGraph.AnimGraphNode_StateMachine_0.Locomotion.AnimStateNode_2.Closed
  nodes [
   n1 = Get OwnerGlider()
   Output Animation Pose(Result=Blend Poses by bool(
     BlendPose_0=Sequence Player 'PuzzleShed_Closed_FrontEnd_GLIDER'() {sequence=&/PuzzleShed_Closed_FrontEnd_GLIDER}
     BlendPose_1=Sequence Player 'PuzzleShed_Closed_GLIDER'() {sequence=&/PuzzleShed_Closed_GLIDER}
     bActiveValue=Get IsCosmeticPreview(self=@n1)
    ) {blendType=Linear})
  ]
 }
 Open_Into {
  path=AnimGraph.AnimGraphNode_StateMachine_0.Locomotion.AnimStateNode_3.Open_Into
  nodes [
   n1 = Break Fort Anim Input Skydiving(FortAnimInput_Skydiving=Get SkydivingState()) -> (
    ParaGilderRootOffset=~
    ParaGliderRootRotationOverride=~
    bUseParaGlideRootModifier=~
    bIsSkydivingFromLaunchPad=~
    bIsSkydivingFromBus=~
    bIsInVortex=~
    bIsDiving=~
    bIsParachuteOpen=~
    bIsSkydiving=~
    bIsParachuteLeaning=~
    bIsSkydiveLeaning=~
    bIsLeaning=~
    bIsSkydiveDiveMode=~
    bPlaySkydiveDrift=~
    bSkydiveDriftDelayActive=~
    bSkydiveDriftAnimAllowed=~
    bIsGliderRight=~
    bIsGliderCenter=~
    bIsGliderLeft=~
    bIsGliderForward=~
    bIsGliderBack=~
    bHasDeployedAndIsStandingUpright=~
    LocalAccelForward=~
    LocalAccelRight=~
    LocalVelocityRight=~
    SkydiveAimPitch=~
    SkydiveAimYaw=~
    DeployChuteAnimRate=~
    SkydiveDriftAnimRate=~
    SkydiveDriftAnimRateCurrent=~
    SkydiveFidgetAnimRate=~
    SkydiveFidgetAnimRateCurrent=~
    SkydiveAdditiveAlpha=~
    SkydiveDriftDelay=~
   )
   n2 = Get OwnerGlider()
   Output Animation Pose(Result=Blend Poses by bool(
     BlendPose_0=Sequence Player 'PuzzleShed_Into_FrontEnd_GLIDER'() {bLoopAnimation=false, sequence=&/PuzzleShed_Into_FrontEnd_GLIDER}
     BlendPose_1=Blendspace Player 'BS_PuzzleShed_Into_GLIDER'(X=@n1.SkydiveAimYaw, Y=@n1.SkydiveAimPitch) {bLoop=false, blendSpace=&/BS_PuzzleShed_Into_GLIDER}
     bActiveValue=Get IsCosmeticPreview(self=@n2)
    ) {blendType=Linear})
  ]
 }
 Transition {
  path=AnimGraph.AnimGraphNode_StateMachine_0.Locomotion.AnimStateTransitionNode_0.Transition
  nodes [
   Result()
   "Equal (Float)"!(A:double="Get Relevant Anim Time Remaining (Open_Into)"!() -> (ReturnValue:float) {is=title K2Node_AnimGetter}, B:double=0) -> (ReturnValue:bool) {is=title K2Node_CallFunction}
  ]
 }
 Transition {
  path=AnimGraph.AnimGraphNode_StateMachine_0.Locomotion.AnimStateTransitionNode_1.Transition
  nodes [
   n1 = Get SkydivingState()
   Result(bCanEnterTransition=OR Boolean(
     A=Break Fort Anim Input Skydiving(FortAnimInput_Skydiving=@n1) -> (
      ParaGilderRootOffset=~
      ParaGliderRootRotationOverride=~
      bUseParaGlideRootModifier=~
      bIsSkydivingFromLaunchPad=~
      bIsSkydivingFromBus=~
      bIsInVortex=~
      bIsDiving=~
      bIsParachuteOpen=~
      bIsSkydiving=~
      bIsParachuteLeaning=~
      bIsSkydiveLeaning=~
      bIsLeaning=~
      bIsSkydiveDiveMode=~
      bPlaySkydiveDrift=~
      bSkydiveDriftDelayActive=~
      bSkydiveDriftAnimAllowed=~
      bIsGliderRight=~
      bIsGliderCenter=~
      bIsGliderLeft=~
      bIsGliderForward=~
      bIsGliderBack=~
      bHasDeployedAndIsStandingUpright=~
      LocalAccelForward=~
      LocalAccelRight=~
      LocalVelocityRight=~
      SkydiveAimPitch=~
      SkydiveAimYaw=~
      DeployChuteAnimRate=~
      SkydiveDriftAnimRate=~
      SkydiveDriftAnimRateCurrent=~
      SkydiveFidgetAnimRate=~
      SkydiveFidgetAnimRateCurrent=~
      SkydiveAdditiveAlpha=~
      SkydiveDriftDelay=~
     ).bIsParachuteOpen
     B=Get bForceOpen() -> (bForceOpen:bool)
    ))
  ]
 }
 Transition {
  path=AnimGraph.AnimGraphNode_StateMachine_0.Locomotion.AnimStateTransitionNode_2.Transition
  nodes [
   n1 = Get bForceOpen()
   n2 = Break Fort Anim Input Skydiving(FortAnimInput_Skydiving=Get SkydivingState()) -> (
    ParaGilderRootOffset=~
    ParaGliderRootRotationOverride=~
    bUseParaGlideRootModifier=~
    bIsSkydivingFromLaunchPad=~
    bIsSkydivingFromBus=~
    bIsInVortex=~
    bIsDiving=~
    bIsParachuteOpen=~
    bIsSkydiving=~
    bIsParachuteLeaning=~
    bIsSkydiveLeaning=~
    bIsLeaning=~
    bIsSkydiveDiveMode=~
    bPlaySkydiveDrift=~
    bSkydiveDriftDelayActive=~
    bSkydiveDriftAnimAllowed=~
    bIsGliderRight=~
    bIsGliderCenter=~
    bIsGliderLeft=~
    bIsGliderForward=~
    bIsGliderBack=~
    bHasDeployedAndIsStandingUpright=~
    LocalAccelForward=~
    LocalAccelRight=~
    LocalVelocityRight=~
    SkydiveAimPitch=~
    SkydiveAimYaw=~
    DeployChuteAnimRate=~
    SkydiveDriftAnimRate=~
    SkydiveDriftAnimRateCurrent=~
    SkydiveFidgetAnimRate=~
    SkydiveFidgetAnimRateCurrent=~
    SkydiveAdditiveAlpha=~
    SkydiveDriftDelay=~
   )
   Result(bCanEnterTransition=AND Boolean(A=NOT Boolean(A=@n2.bIsParachuteOpen), B=NOT Boolean(A=@n1)))
  ]
 }
 Transition {
  path=AnimGraph.AnimGraphNode_StateMachine_0.Locomotion.AnimStateTransitionNode_3.Transition
  nodes [
   n1 = Get bForceOpen()
   n2 = Break Fort Anim Input Skydiving(FortAnimInput_Skydiving=Get SkydivingState()) -> (
    ParaGilderRootOffset=~
    ParaGliderRootRotationOverride=~
    bUseParaGlideRootModifier=~
    bIsSkydivingFromLaunchPad=~
    bIsSkydivingFromBus=~
    bIsInVortex=~
    bIsDiving=~
    bIsParachuteOpen=~
    bIsSkydiving=~
    bIsParachuteLeaning=~
    bIsSkydiveLeaning=~
    bIsLeaning=~
    bIsSkydiveDiveMode=~
    bPlaySkydiveDrift=~
    bSkydiveDriftDelayActive=~
    bSkydiveDriftAnimAllowed=~
    bIsGliderRight=~
    bIsGliderCenter=~
    bIsGliderLeft=~
    bIsGliderForward=~
    bIsGliderBack=~
    bHasDeployedAndIsStandingUpright=~
    LocalAccelForward=~
    LocalAccelRight=~
    LocalVelocityRight=~
    SkydiveAimPitch=~
    SkydiveAimYaw=~
    DeployChuteAnimRate=~
    SkydiveDriftAnimRate=~
    SkydiveDriftAnimRateCurrent=~
    SkydiveFidgetAnimRate=~
    SkydiveFidgetAnimRateCurrent=~
    SkydiveAdditiveAlpha=~
    SkydiveDriftDelay=~
   )
   Result(bCanEnterTransition=AND Boolean(A=NOT Boolean(A=@n2.bIsParachuteOpen), B=NOT Boolean(A=@n1)))
  ]
 }
 Locomotion_GliderItem {
  entry=Closed
  path=AnimGraph.AnimGraphNode_StateMachine_1.Locomotion_GliderItem
  ids=id
  nodes [
   AnimStateNode_0 = Idle_Open()
   AnimStateNode_1 = ForceOpen_or_GliderItem() {is=title AnimStateNode, stateEntered {guid=#3, notifyName=DeployEntered}}
   AnimStateNode_2 = Closed() {stateEntered {guid=#1, notifyName=ClosedEntered}}
   AnimStateNode_3 = Open_Into() {stateEntered {guid=#2, notifyName=DeployEntered}}
   AnimStateTransitionNode_0 = AnimStateTransitionNode() {bAutomaticRuleBasedOnSequencePlayerInState=true, blendMode=Linear, crossfadeDuration=0}
   AnimStateTransitionNode_1 = AnimStateTransitionNode() {crossfadeDuration=0}
   AnimStateTransitionNode_2 = AnimStateTransitionNode() {crossfadeDuration=0}
   AnimStateTransitionNode_3 = AnimStateTransitionNode() {crossfadeDuration=0, priorityOrder=2}
   AnimStateTransitionNode_4 = AnimStateTransitionNode() {crossfadeDuration=0}
   AnimStateTransitionNode_5 = AnimStateTransitionNode() {bAutomaticRuleBasedOnSequencePlayerInState=true, crossfadeDuration=0}
   AnimStateTransitionNode_6 = AnimStateTransitionNode() {crossfadeDuration=0}
  ]
  links [
   {
    dst=ForceOpen_or_GliderItem
    rule=AnimGraph.AnimGraphNode_StateMachine_1.Locomotion_GliderItem.AnimStateTransitionNode_4.Transition
    src=Closed
    via=AnimStateTransitionNode_4
   }
   {
    dst=Open_Into
    rule=AnimGraph.AnimGraphNode_StateMachine_1.Locomotion_GliderItem.AnimStateTransitionNode_1.Transition
    src=Closed
    via=AnimStateTransitionNode_1
   }
   {
    dst=Closed
    rule=AnimGraph.AnimGraphNode_StateMachine_1.Locomotion_GliderItem.AnimStateTransitionNode_6.Transition
    src=ForceOpen_or_GliderItem
    via=AnimStateTransitionNode_6
   }
   {
    dst=Idle_Open
    rule=AnimGraph.AnimGraphNode_StateMachine_1.Locomotion_GliderItem.AnimStateTransitionNode_5.Transition
    src=ForceOpen_or_GliderItem
    via=AnimStateTransitionNode_5
   }
   {
    dst=Closed
    rule=AnimGraph.AnimGraphNode_StateMachine_1.Locomotion_GliderItem.AnimStateTransitionNode_2.Transition
    src=Idle_Open
    via=AnimStateTransitionNode_2
   }
   {
    dst=Closed
    rule=AnimGraph.AnimGraphNode_StateMachine_1.Locomotion_GliderItem.AnimStateTransitionNode_3.Transition
    src=Open_Into
    via=AnimStateTransitionNode_3
   }
   {
    dst=Idle_Open
    rule=AnimGraph.AnimGraphNode_StateMachine_1.Locomotion_GliderItem.AnimStateTransitionNode_0.Transition
    src=Open_Into
    via=AnimStateTransitionNode_0
   }
  ]
 }
 Idle_Open {
  path=AnimGraph.AnimGraphNode_StateMachine_1.Locomotion_GliderItem.AnimStateNode_0.Idle_Open
  nodes [
   Apply Additive!(Base, Additive, bAlphaBoolEnabled:bool=Get IsCosmeticPreview(self=Get OwnerGlider())) -> (Pose) {alphaInputType=Bool, alphaBoolBlend {blendOption=Linear}}
   n1 = Break Fort Anim Input Skydiving(FortAnimInput_Skydiving=Get SkydivingState()) -> (
    ParaGilderRootOffset=~
    ParaGliderRootRotationOverride=~
    bUseParaGlideRootModifier=~
    bIsSkydivingFromLaunchPad=~
    bIsSkydivingFromBus=~
    bIsInVortex=~
    bIsDiving=~
    bIsParachuteOpen=~
    bIsSkydiving=~
    bIsParachuteLeaning=~
    bIsSkydiveLeaning=~
    bIsLeaning=~
    bIsSkydiveDiveMode=~
    bPlaySkydiveDrift=~
    bSkydiveDriftDelayActive=~
    bSkydiveDriftAnimAllowed=~
    bIsGliderRight=~
    bIsGliderCenter=~
    bIsGliderLeft=~
    bIsGliderForward=~
    bIsGliderBack=~
    bHasDeployedAndIsStandingUpright=~
    LocalAccelForward=~
    LocalAccelRight=~
    LocalVelocityRight=~
    SkydiveAimPitch=~
    SkydiveAimYaw=~
    DeployChuteAnimRate=~
    SkydiveDriftAnimRate=~
    SkydiveDriftAnimRateCurrent=~
    SkydiveFidgetAnimRate=~
    SkydiveFidgetAnimRateCurrent=~
    SkydiveAdditiveAlpha=~
    SkydiveDriftDelay=~
   )
   n2 = Get OwnerGlider()
   Output Animation Pose(Result=Blend Poses by bool(
     BlendPose_0=Sequence Player 'PuzzleShed_Base_Pose_GLIDER'() {sequence=&/PuzzleShed_Base_Pose_GLIDER}
     BlendPose_1=Blendspace Player 'BS_PuzzleShed_GLIDER'(X=@n1.SkydiveAimYaw, Y=@n1.SkydiveAimPitch) {blendSpace=&BS_PuzzleShed_GLIDER}
     bActiveValue=Get IsCosmeticPreview(self=@n2)
    ) {blendType=Linear})
  ]
 }
 ForceOpen_or_GliderItem {
  path=AnimGraph.AnimGraphNode_StateMachine_1.Locomotion_GliderItem.AnimStateNode_1.ForceOpen_or_GliderItem
  nodes [
   n1 = Break Fort Anim Input Skydiving(FortAnimInput_Skydiving=Get SkydivingState()) -> (
    ParaGilderRootOffset=~
    ParaGliderRootRotationOverride=~
    bUseParaGlideRootModifier=~
    bIsSkydivingFromLaunchPad=~
    bIsSkydivingFromBus=~
    bIsInVortex=~
    bIsDiving=~
    bIsParachuteOpen=~
    bIsSkydiving=~
    bIsParachuteLeaning=~
    bIsSkydiveLeaning=~
    bIsLeaning=~
    bIsSkydiveDiveMode=~
    bPlaySkydiveDrift=~
    bSkydiveDriftDelayActive=~
    bSkydiveDriftAnimAllowed=~
    bIsGliderRight=~
    bIsGliderCenter=~
    bIsGliderLeft=~
    bIsGliderForward=~
    bIsGliderBack=~
    bHasDeployedAndIsStandingUpright=~
    LocalAccelForward=~
    LocalAccelRight=~
    LocalVelocityRight=~
    SkydiveAimPitch=~
    SkydiveAimYaw=~
    DeployChuteAnimRate=~
    SkydiveDriftAnimRate=~
    SkydiveDriftAnimRateCurrent=~
    SkydiveFidgetAnimRate=~
    SkydiveFidgetAnimRateCurrent=~
    SkydiveAdditiveAlpha=~
    SkydiveDriftDelay=~
   )
   n2 = Blendspace Player 'BS_PuzzleShed_Into_GLIDER'(X=@n1.SkydiveAimYaw, Y=@n1.SkydiveAimPitch) {bLoop=false, blendSpace=&BS_PuzzleShed_Into_GLIDER, startPosition=0.0044}
   Output Animation Pose(Result=@n2)
  ]
 }
 Closed {
  path=AnimGraph.AnimGraphNode_StateMachine_1.Locomotion_GliderItem.AnimStateNode_2.Closed
  nodes [
   n1 = Get OwnerGlider()
   Output Animation Pose(Result=Blend Poses by bool(
     BlendPose_0=Sequence Player 'PuzzleShed_Closed_GLIDER'() {sequence=&PuzzleShed_Closed_GLIDER}
     BlendPose_1=Sequence Player 'PuzzleShed_Closed_GLIDER'() {sequence=&PuzzleShed_Closed_GLIDER}
     bActiveValue=Get IsCosmeticPreview(self=@n1)
    ) {blendType=Linear})
  ]
 }
 Open_Into {
  path=AnimGraph.AnimGraphNode_StateMachine_1.Locomotion_GliderItem.AnimStateNode_3.Open_Into
  nodes [
   Apply Additive(bAlphaBoolEnabled=Get IsCosmeticPreview(self=Get OwnerGlider())) {alphaInputType=Bool, alphaBoolBlend {blendOption=Linear}}
   n1 = Break Fort Anim Input Skydiving(FortAnimInput_Skydiving=Get SkydivingState()) -> (
    ParaGilderRootOffset=~
    ParaGliderRootRotationOverride=~
    bUseParaGlideRootModifier=~
    bIsSkydivingFromLaunchPad=~
    bIsSkydivingFromBus=~
    bIsInVortex=~
    bIsDiving=~
    bIsParachuteOpen=~
    bIsSkydiving=~
    bIsParachuteLeaning=~
    bIsSkydiveLeaning=~
    bIsLeaning=~
    bIsSkydiveDiveMode=~
    bPlaySkydiveDrift=~
    bSkydiveDriftDelayActive=~
    bSkydiveDriftAnimAllowed=~
    bIsGliderRight=~
    bIsGliderCenter=~
    bIsGliderLeft=~
    bIsGliderForward=~
    bIsGliderBack=~
    bHasDeployedAndIsStandingUpright=~
    LocalAccelForward=~
    LocalAccelRight=~
    LocalVelocityRight=~
    SkydiveAimPitch=~
    SkydiveAimYaw=~
    DeployChuteAnimRate=~
    SkydiveDriftAnimRate=~
    SkydiveDriftAnimRateCurrent=~
    SkydiveFidgetAnimRate=~
    SkydiveFidgetAnimRateCurrent=~
    SkydiveAdditiveAlpha=~
    SkydiveDriftDelay=~
   )
   n2 = Get OwnerGlider()
   Output Animation Pose(Result=Blend Poses by bool(
     BlendPose_0=Sequence Player 'PuzzleShed_Into_GLIDER'() {bLoopAnimation=false, sequence=&/PuzzleShed_Into_GLIDER}
     BlendPose_1=Blendspace Player 'BS_PuzzleShed_Into_GLIDER'(X=@n1.SkydiveAimYaw, Y=@n1.SkydiveAimPitch) {bLoop=false, blendSpace=&BS_PuzzleShed_Into_GLIDER}
     bActiveValue=Get IsCosmeticPreview(self=@n2)
    ) {blendType=Linear})
  ]
 }
 Transition {
  path=AnimGraph.AnimGraphNode_StateMachine_1.Locomotion_GliderItem.AnimStateTransitionNode_0.Transition
  nodes [
   Result()
   "Equal (Float)"(A="Get Relevant Anim Time Remaining (Open_Into)"(), B=0)
  ]
 }
 Transition {
  path=AnimGraph.AnimGraphNode_StateMachine_1.Locomotion_GliderItem.AnimStateTransitionNode_1.Transition
  nodes [
   n1 = Break Fort Anim Input Skydiving(FortAnimInput_Skydiving=Get SkydivingState()) -> (
    ParaGilderRootOffset=~
    ParaGliderRootRotationOverride=~
    bUseParaGlideRootModifier=~
    bIsSkydivingFromLaunchPad=~
    bIsSkydivingFromBus=~
    bIsInVortex=~
    bIsDiving=~
    bIsParachuteOpen=~
    bIsSkydiving=~
    bIsParachuteLeaning=~
    bIsSkydiveLeaning=~
    bIsLeaning=~
    bIsSkydiveDiveMode=~
    bPlaySkydiveDrift=~
    bSkydiveDriftDelayActive=~
    bSkydiveDriftAnimAllowed=~
    bIsGliderRight=~
    bIsGliderCenter=~
    bIsGliderLeft=~
    bIsGliderForward=~
    bIsGliderBack=~
    bHasDeployedAndIsStandingUpright=~
    LocalAccelForward=~
    LocalAccelRight=~
    LocalVelocityRight=~
    SkydiveAimPitch=~
    SkydiveAimYaw=~
    DeployChuteAnimRate=~
    SkydiveDriftAnimRate=~
    SkydiveDriftAnimRateCurrent=~
    SkydiveFidgetAnimRate=~
    SkydiveFidgetAnimRateCurrent=~
    SkydiveAdditiveAlpha=~
    SkydiveDriftDelay=~
   )
   n2 = Get IsCosmeticPreview(self=Get OwnerGlider())
   Result(bCanEnterTransition=AND Boolean(
     A=OR Boolean!(A=@n1.bIsSkydivingFromLaunchPad, B=@n1.bIsSkydivingFromBus, C:bool=@n1.bIsInVortex, D:bool=@n2) -> (ReturnValue)
     B=@n1.bIsParachuteOpen
    ))
  ]
 }
 Transition {
  path=AnimGraph.AnimGraphNode_StateMachine_1.Locomotion_GliderItem.AnimStateTransitionNode_2.Transition
  nodes [
   n1 = Get bForceOpen()
   n2 = Break Fort Anim Input Skydiving(FortAnimInput_Skydiving=Get SkydivingState()) -> (
    ParaGilderRootOffset=~
    ParaGliderRootRotationOverride=~
    bUseParaGlideRootModifier=~
    bIsSkydivingFromLaunchPad=~
    bIsSkydivingFromBus=~
    bIsInVortex=~
    bIsDiving=~
    bIsParachuteOpen=~
    bIsSkydiving=~
    bIsParachuteLeaning=~
    bIsSkydiveLeaning=~
    bIsLeaning=~
    bIsSkydiveDiveMode=~
    bPlaySkydiveDrift=~
    bSkydiveDriftDelayActive=~
    bSkydiveDriftAnimAllowed=~
    bIsGliderRight=~
    bIsGliderCenter=~
    bIsGliderLeft=~
    bIsGliderForward=~
    bIsGliderBack=~
    bHasDeployedAndIsStandingUpright=~
    LocalAccelForward=~
    LocalAccelRight=~
    LocalVelocityRight=~
    SkydiveAimPitch=~
    SkydiveAimYaw=~
    DeployChuteAnimRate=~
    SkydiveDriftAnimRate=~
    SkydiveDriftAnimRateCurrent=~
    SkydiveFidgetAnimRate=~
    SkydiveFidgetAnimRateCurrent=~
    SkydiveAdditiveAlpha=~
    SkydiveDriftDelay=~
   )
   Result(bCanEnterTransition=AND Boolean(A=NOT Boolean(A=@n2.bIsParachuteOpen), B=NOT Boolean(A=@n1)))
  ]
 }
 Transition {
  path=AnimGraph.AnimGraphNode_StateMachine_1.Locomotion_GliderItem.AnimStateTransitionNode_3.Transition
  nodes [
   n1 = Get bForceOpen()
   n2 = Break Fort Anim Input Skydiving(FortAnimInput_Skydiving=Get SkydivingState()) -> (
    ParaGilderRootOffset=~
    ParaGliderRootRotationOverride=~
    bUseParaGlideRootModifier=~
    bIsSkydivingFromLaunchPad=~
    bIsSkydivingFromBus=~
    bIsInVortex=~
    bIsDiving=~
    bIsParachuteOpen=~
    bIsSkydiving=~
    bIsParachuteLeaning=~
    bIsSkydiveLeaning=~
    bIsLeaning=~
    bIsSkydiveDiveMode=~
    bPlaySkydiveDrift=~
    bSkydiveDriftDelayActive=~
    bSkydiveDriftAnimAllowed=~
    bIsGliderRight=~
    bIsGliderCenter=~
    bIsGliderLeft=~
    bIsGliderForward=~
    bIsGliderBack=~
    bHasDeployedAndIsStandingUpright=~
    LocalAccelForward=~
    LocalAccelRight=~
    LocalVelocityRight=~
    SkydiveAimPitch=~
    SkydiveAimYaw=~
    DeployChuteAnimRate=~
    SkydiveDriftAnimRate=~
    SkydiveDriftAnimRateCurrent=~
    SkydiveFidgetAnimRate=~
    SkydiveFidgetAnimRateCurrent=~
    SkydiveAdditiveAlpha=~
    SkydiveDriftDelay=~
   )
   Result(bCanEnterTransition=AND Boolean(A=NOT Boolean(A=@n2.bIsParachuteOpen), B=NOT Boolean(A=@n1)))
  ]
 }
 Transition {
  path=AnimGraph.AnimGraphNode_StateMachine_1.Locomotion_GliderItem.AnimStateTransitionNode_4.Transition
  nodes [
   n1 = Break Fort Anim Input Skydiving(FortAnimInput_Skydiving=Get SkydivingState()) -> (
    ParaGilderRootOffset=~
    ParaGliderRootRotationOverride=~
    bUseParaGlideRootModifier=~
    bIsSkydivingFromLaunchPad=~
    bIsSkydivingFromBus=~
    bIsInVortex=~
    bIsDiving=~
    bIsParachuteOpen=~
    bIsSkydiving=~
    bIsParachuteLeaning=~
    bIsSkydiveLeaning=~
    bIsLeaning=~
    bIsSkydiveDiveMode=~
    bPlaySkydiveDrift=~
    bSkydiveDriftDelayActive=~
    bSkydiveDriftAnimAllowed=~
    bIsGliderRight=~
    bIsGliderCenter=~
    bIsGliderLeft=~
    bIsGliderForward=~
    bIsGliderBack=~
    bHasDeployedAndIsStandingUpright=~
    LocalAccelForward=~
    LocalAccelRight=~
    LocalVelocityRight=~
    SkydiveAimPitch=~
    SkydiveAimYaw=~
    DeployChuteAnimRate=~
    SkydiveDriftAnimRate=~
    SkydiveDriftAnimRateCurrent=~
    SkydiveFidgetAnimRate=~
    SkydiveFidgetAnimRateCurrent=~
    SkydiveAdditiveAlpha=~
    SkydiveDriftDelay=~
   )
   Result(bCanEnterTransition=AND Boolean!(
     A=NOT Boolean(A=@n1.bIsSkydivingFromLaunchPad)
     B=NOT Boolean(A=@n1.bIsSkydivingFromBus)
     C:bool=NOT Boolean(A=@n1.bIsInVortex)
     D:bool=Get bForceOpen()
    ) -> (ReturnValue))
  ]
 }
 Transition {path=AnimGraph.AnimGraphNode_StateMachine_1.Locomotion_GliderItem.AnimStateTransitionNode_5.Transition, nodes [Result()]}
 Transition {
  path=AnimGraph.AnimGraphNode_StateMachine_1.Locomotion_GliderItem.AnimStateTransitionNode_6.Transition
  nodes [
   n1 = Get bForceOpen()
   n2 = Break Fort Anim Input Skydiving(FortAnimInput_Skydiving=Get SkydivingState()) -> (
    ParaGilderRootOffset=~
    ParaGliderRootRotationOverride=~
    bUseParaGlideRootModifier=~
    bIsSkydivingFromLaunchPad=~
    bIsSkydivingFromBus=~
    bIsInVortex=~
    bIsDiving=~
    bIsParachuteOpen=~
    bIsSkydiving=~
    bIsParachuteLeaning=~
    bIsSkydiveLeaning=~
    bIsLeaning=~
    bIsSkydiveDiveMode=~
    bPlaySkydiveDrift=~
    bSkydiveDriftDelayActive=~
    bSkydiveDriftAnimAllowed=~
    bIsGliderRight=~
    bIsGliderCenter=~
    bIsGliderLeft=~
    bIsGliderForward=~
    bIsGliderBack=~
    bHasDeployedAndIsStandingUpright=~
    LocalAccelForward=~
    LocalAccelRight=~
    LocalVelocityRight=~
    SkydiveAimPitch=~
    SkydiveAimYaw=~
    DeployChuteAnimRate=~
    SkydiveDriftAnimRate=~
    SkydiveDriftAnimRateCurrent=~
    SkydiveFidgetAnimRate=~
    SkydiveFidgetAnimRateCurrent=~
    SkydiveAdditiveAlpha=~
    SkydiveDriftDelay=~
   )
   Result(bCanEnterTransition=AND Boolean!(A=NOT Boolean(A=@n2.bIsParachuteOpen), B=NOT Boolean(A=@n1)) -> (ReturnValue))
  ]
 }
 Blink {
  entry=Identity
  path=AnimGraph.AnimGraphNode_StateMachine_2.Blink
  ids=id
  nodes [
   AnimStateNode_0 = Identity() {is=title AnimStateNode}
   AnimStateNode_1 = Blink() {is=title AnimStateNode}
   AnimStateTransitionNode_0 = AnimStateTransitionNode()
   AnimStateTransitionNode_1 = AnimStateTransitionNode() {bAutomaticRuleBasedOnSequencePlayerInState=true}
  ]
  links [
   {
    dst=Identity
    rule=AnimGraph.AnimGraphNode_StateMachine_2.Blink.AnimStateTransitionNode_1.Transition
    src=Blink
    via=AnimStateTransitionNode_1
   }
   {
    dst=Blink
    rule=AnimGraph.AnimGraphNode_StateMachine_2.Blink.AnimStateTransitionNode_0.Transition
    src=Identity
    via=AnimStateTransitionNode_0
   }
  ]
 }
 Identity {path=AnimGraph.AnimGraphNode_StateMachine_2.Blink.AnimStateNode_0.Identity, nodes [Output Animation Pose(Result=Additive Identity Pose())]}
 Blink {
  path=AnimGraph.AnimGraphNode_StateMachine_2.Blink.AnimStateNode_1.Blink
  nodes [
   n1 = "Sequence Player (additive) 'PuzzleShed_Blink_Frontend_Anim_GLIDER'"() {bLoopAnimation=false, sequence=&/PuzzleShed_Blink_Frontend_Anim_GLIDER}
   Output Animation Pose(Result=Blend Poses by bool(
     BlendPose_0=@n1
     BlendPose_1="Sequence Player (additive) 'PuzzleShed_Blink_Anim_GLIDER'"() {bLoopAnimation=false, sequence=&/PuzzleShed_Blink_Anim_GLIDER}
     bActiveValue=Get IsInFrontEnd()
    ) {blendType=Linear})
  ]
 }
 Transition {
  path=AnimGraph.AnimGraphNode_StateMachine_2.Blink.AnimStateTransitionNode_0.Transition
  nodes [Result(bCanEnterTransition=Get IsBlinking() -> (IsBlinking:bool))]
 }
 Transition {path=AnimGraph.AnimGraphNode_StateMachine_2.Blink.AnimStateTransitionNode_1.Transition, nodes [Result()]}
 Jitter {entry=Jitter, path=AnimGraph.AnimGraphNode_StateMachine_3.Jitter, ids=id, nodes [AnimStateNode_1 = Jitter() {is=title AnimStateNode}]}
 Jitter {
  path=AnimGraph.AnimGraphNode_StateMachine_3.Jitter.AnimStateNode_1.Jitter
  nodes [
   n1 = Get Acceleration() -> (Acceleration:double)
   Output Animation Pose(Result=Blend Poses by bool(
     BlendPose_0="Sequence Player (additive) 'PuzzleShed_Jitter_FrontEnd_Anim_GLIDER'"() {sequence=&/PuzzleShed_Jitter_FrontEnd_Anim_GLIDER}
     BlendPose_1="Sequence Player (additive) 'PuzzleShed_Jitter_GLIDER'"!(PlayRate:float=@n1) -> (Pose) {sequence=&/PuzzleShed_Jitter_GLIDER}
     bActiveValue=Get IsInFrontEnd()
    ) {blendType=Linear})
  ]
 }
 Turn Transition {
  entry=Identify
  path=AnimGraph.AnimGraphNode_StateMachine_4.Turn Transition
  ids=id
  nodes [
   AnimStateNode_14 = Left Into() {
    is=title AnimStateNode
    stateEntered {guid=#4}
    stateFullyBlended {guid=#5}
    stateLeft {guid=#6}
   }
   AnimStateNode_16 = Right Into() {
    is=title AnimStateNode
    stateEntered {guid=#4}
    stateFullyBlended {guid=#5}
    stateLeft {guid=#6}
   }
   AnimStateNode_7 = Identify() {
    is=title AnimStateNode
    stateEntered {guid=#7}
    stateFullyBlended {guid=#8}
    stateLeft {guid=#9}
   }
   AnimStateTransitionNode_58 = AnimStateTransitionNode() {crossfadeDuration=0.5}
   AnimStateTransitionNode_59 = AnimStateTransitionNode() {crossfadeDuration=0.5}
   AnimStateTransitionNode_60 = AnimStateTransitionNode()
   AnimStateTransitionNode_61 = AnimStateTransitionNode()
   AnimStateTransitionNode_62 = AnimStateTransitionNode() {crossfadeDuration=0.5}
   AnimStateTransitionNode_63 = AnimStateTransitionNode() {crossfadeDuration=0.5}
  ]
  links [
   {
    dst=Left Into
    rule=AnimGraph.AnimGraphNode_StateMachine_4.Turn Transition.AnimStateTransitionNode_58.Transition
    src=Identify
    via=AnimStateTransitionNode_58
   }
   {
    dst=Right Into
    rule=AnimGraph.AnimGraphNode_StateMachine_4.Turn Transition.AnimStateTransitionNode_59.Transition
    src=Identify
    via=AnimStateTransitionNode_59
   }
   {
    dst=Identify
    rule=AnimGraph.AnimGraphNode_StateMachine_4.Turn Transition.AnimStateTransitionNode_60.Transition
    src=Left Into
    via=AnimStateTransitionNode_60
   }
   {
    dst=Identify
    rule=AnimGraph.AnimGraphNode_StateMachine_4.Turn Transition.AnimStateTransitionNode_62.Transition
    src=Left Into
    via=AnimStateTransitionNode_62
   }
   {
    dst=Identify
    rule=AnimGraph.AnimGraphNode_StateMachine_4.Turn Transition.AnimStateTransitionNode_61.Transition
    src=Right Into
    via=AnimStateTransitionNode_61
   }
   {
    dst=Identify
    rule=AnimGraph.AnimGraphNode_StateMachine_4.Turn Transition.AnimStateTransitionNode_63.Transition
    src=Right Into
    via=AnimStateTransitionNode_63
   }
  ]
 }
 Left Into {
  path=AnimGraph.AnimGraphNode_StateMachine_4.Turn Transition.AnimStateNode_14.Left Into
  nodes [
   n1 = "Sequence Player (additive) 'PuzzleShed_Lean_Left_Into_Anim_GLIDER'"!(PlayRate:float=Get Yaw Velocity() -> (Yaw Velocity:double)) -> (Pose) {bLoopAnimation=false, sequence=&/PuzzleShed_Lean_Left_Into_Anim_GLIDER}
   Output Animation Pose(Result=@n1)
  ]
 }
 Right Into {
  path=AnimGraph.AnimGraphNode_StateMachine_4.Turn Transition.AnimStateNode_16.Right Into
  nodes [
   n1 = "Sequence Player (additive) 'PuzzleShed_Lean_Right_Into_Anim_GLIDER'"!(PlayRate:float=Get Yaw Velocity()) -> (Pose) {bLoopAnimation=false, sequence=&/PuzzleShed_Lean_Right_Into_Anim_GLIDER}
   Output Animation Pose(Result=@n1)
  ]
 }
 Identify {
  path=AnimGraph.AnimGraphNode_StateMachine_4.Turn Transition.AnimStateNode_7.Identify
  nodes [
   n1 = "Sequence Player (additive) 'PuzzleShed_Lean_Center_Pose_GLIDER'"() {bLoopAnimation=false, sequence=&/PuzzleShed_Lean_Center_Pose_GLIDER}
   Output Animation Pose(Result=@n1)
  ]
 }
 Transition {
  path=AnimGraph.AnimGraphNode_StateMachine_4.Turn Transition.AnimStateTransitionNode_58.Transition
  nodes [Result(bCanEnterTransition=AND Boolean(A=Get Glide_Left() -> (Glide_Left:bool), B=Get Yaw Input() -> (Yaw Input:bool)))]
 }
 Transition {
  path=AnimGraph.AnimGraphNode_StateMachine_4.Turn Transition.AnimStateTransitionNode_59.Transition
  nodes [Result(bCanEnterTransition=AND Boolean(A=Get Glide_right() -> (Glide_right:bool), B=Get Yaw Input()))]
 }
 Transition {
  path=AnimGraph.AnimGraphNode_StateMachine_4.Turn Transition.AnimStateTransitionNode_60.Transition
  nodes [Result(bCanEnterTransition=AND Boolean(A=Get Glide_right(), B=Get Yaw Input()))]
 }
 Transition {
  path=AnimGraph.AnimGraphNode_StateMachine_4.Turn Transition.AnimStateTransitionNode_61.Transition
  nodes [Result(bCanEnterTransition=AND Boolean(A=Get Glide_Left(), B=Get Yaw Input()))]
 }
 Transition {
  path=AnimGraph.AnimGraphNode_StateMachine_4.Turn Transition.AnimStateTransitionNode_62.Transition
  nodes [Result(bCanEnterTransition=NOT Boolean(A=Get Yaw Input()))]
 }
 Transition {
  path=AnimGraph.AnimGraphNode_StateMachine_4.Turn Transition.AnimStateTransitionNode_63.Transition
  nodes [Result(bCanEnterTransition=NOT Boolean(A=Get Yaw Input()))]
 }
 Old_Turn {entry=Turn, path=AnimGraph.AnimGraphNode_StateMachine_5.Old_Turn, ids=id, nodes [AnimStateNode_0 = Turn() {is=title AnimStateNode}]}
 Turn {
  path=AnimGraph.AnimGraphNode_StateMachine_5.Old_Turn.AnimStateNode_0.Turn
  nodes [
   n1 = Break Fort Anim Input Skydiving(FortAnimInput_Skydiving=Get SkydivingState()) -> (
    ParaGilderRootOffset=~
    ParaGliderRootRotationOverride=~
    bUseParaGlideRootModifier=~
    bIsSkydivingFromLaunchPad=~
    bIsSkydivingFromBus=~
    bIsInVortex=~
    bIsDiving=~
    bIsParachuteOpen=~
    bIsSkydiving=~
    bIsParachuteLeaning=~
    bIsSkydiveLeaning=~
    bIsLeaning=~
    bIsSkydiveDiveMode=~
    bPlaySkydiveDrift=~
    bSkydiveDriftDelayActive=~
    bSkydiveDriftAnimAllowed=~
    bIsGliderRight=~
    bIsGliderCenter=~
    bIsGliderLeft=~
    bIsGliderForward=~
    bIsGliderBack=~
    bHasDeployedAndIsStandingUpright=~
    LocalAccelForward=~
    LocalAccelRight=~
    LocalVelocityRight=~
    SkydiveAimPitch=~
    SkydiveAimYaw=~
    DeployChuteAnimRate=~
    SkydiveDriftAnimRate=~
    SkydiveDriftAnimRateCurrent=~
    SkydiveFidgetAnimRate=~
    SkydiveFidgetAnimRateCurrent=~
    SkydiveAdditiveAlpha=~
    SkydiveDriftDelay=~
   )
   Output Animation Pose(Result=Blendspace Player 'BS_PuzzleShed_Turn_GLIDER'(X=@n1.SkydiveAimYaw, Y=@n1.SkydiveAimPitch) {blendSpace=&/BS_PuzzleShed_Turn_GLIDER})
  ]
 }
 Lean Transition {
  entry=Lean_Zero
  path=AnimGraph.AnimGraphNode_StateMachine_6.Lean Transition
  ids=id
  nodes [
   AnimStateNode_10 = Lean_Zero() {
    is=title AnimStateNode
    stateEntered {guid=#10}
    stateFullyBlended {guid=#11}
    stateLeft {guid=#12}
   }
   AnimStateNode_14 = Lean_Out() {
    is=title AnimStateNode
    stateEntered {guid=#13}
    stateFullyBlended {guid=#14}
    stateLeft {guid=#15}
   }
   AnimStateNode_17 = Lean_Into() {
    is=title AnimStateNode
    stateEntered {guid=#16}
    stateFullyBlended {guid=#17}
    stateLeft {guid=#18}
   }
   AnimStateTransitionNode_26 = AnimStateTransitionNode() {bAutomaticRuleBasedOnSequencePlayerInState=true}
   AnimStateTransitionNode_37 = AnimStateTransitionNode()
   AnimStateTransitionNode_38 = AnimStateTransitionNode()
   AnimStateTransitionNode_40 = AnimStateTransitionNode()
  ]
  links [
   {
    dst=Lean_Out
    rule=AnimGraph.AnimGraphNode_StateMachine_6.Lean Transition.AnimStateTransitionNode_38.AnimationTransitionGraph_1
    src=Lean_Into
    via=AnimStateTransitionNode_38
   }
   {
    dst=Lean_Into
    rule=AnimGraph.AnimGraphNode_StateMachine_6.Lean Transition.AnimStateTransitionNode_40.AnimationTransitionGraph_1
    src=Lean_Out
    via=AnimStateTransitionNode_40
   }
   {
    dst=Lean_Zero
    rule=AnimGraph.AnimGraphNode_StateMachine_6.Lean Transition.AnimStateTransitionNode_26.AnimationTransitionGraph_1
    src=Lean_Out
    via=AnimStateTransitionNode_26
   }
   {
    dst=Lean_Into
    rule=AnimGraph.AnimGraphNode_StateMachine_6.Lean Transition.AnimStateTransitionNode_37.AnimationTransitionGraph_1
    src=Lean_Zero
    via=AnimStateTransitionNode_37
   }
  ]
 }
 Lean_Zero {
  path=AnimGraph.AnimGraphNode_StateMachine_6.Lean Transition.AnimStateNode_10.Lean_Zero
  nodes [
   n1 = "Sequence Player (additive) 'PuzzleShed_Lean_Center_Pose_GLIDER'"() {bLoopAnimation=false, sequence=&PuzzleShed_Lean_Center_Pose_GLIDER}
   Output Animation Pose(Result=@n1)
  ]
 }
 Lean_Out {
  path=AnimGraph.AnimGraphNode_StateMachine_6.Lean Transition.AnimStateNode_14.Lean_Out
  nodes [
   n1 = "Sequence Player (additive) 'PuzzleShed_Lean_Forward_Out_Anim_GLIDER'"() {bLoopAnimation=false, sequence=&/PuzzleShed_Lean_Forward_Out_Anim_GLIDER}
   n2 = "Sequence Player (additive) 'PuzzleShed_Lean_Back_Out_Anim_GLIDER'"() {bLoopAnimation=false, sequence=&/PuzzleShed_Lean_Back_Out_Anim_GLIDER}
   n3 = "Sequence Player (additive) 'PuzzleShed_Lean_Center_Pose_GLIDER'"() {bLoopAnimation=false, sequence=&PuzzleShed_Lean_Center_Pose_GLIDER}
   n4 = "Blend Poses (ESkydivingDirection)"!(
    BlendPose_0:Pose Link Structure=@n3
    BlendPose_1:Pose Link Structure=@n1
    BlendPose_2:Pose Link Structure=@n2
    BlendPose_3:Pose Link Structure
    BlendPose_4:Pose Link Structure
    BlendTime_0:float=1
    BlendTime_1:float=1
    BlendTime_2:float=1
    BlendTime_3:float=1
    BlendTime_4:float=1
    ActiveEnumValue:Byte=Break Fort Anim Input Skydiving(FortAnimInput_Skydiving=Get SkydivingState()) -> (
     ParaGilderRootOffset=~
     ParaGliderRootRotationOverride=~
     bUseParaGlideRootModifier=~
     bIsSkydivingFromLaunchPad=~
     bIsSkydivingFromBus=~
     bIsInVortex=~
     bIsDiving=~
     bIsParachuteOpen=~
     bIsSkydiving=~
     bIsParachuteLeaning=~
     bIsSkydiveLeaning=~
     bIsLeaning=~
     bIsSkydiveDiveMode=~
     bPlaySkydiveDrift=~
     bSkydiveDriftDelayActive=~
     bSkydiveDriftAnimAllowed=~
     bIsGliderRight=~
     bIsGliderCenter=~
     bIsGliderLeft=~
     bIsGliderForward=~
     bIsGliderBack=~
     bHasDeployedAndIsStandingUpright=~
     LocalAccelForward=~
     LocalAccelRight=~
     LocalVelocityRight=~
     SkydiveAimPitch=~
     SkydiveAimYaw=~
     DeployChuteAnimRate=~
     SkydiveDriftAnimRate=~
     SkydiveDriftAnimRateCurrent=~
     SkydiveFidgetAnimRate=~
     SkydiveFidgetAnimRateCurrent=~
     SkydiveAdditiveAlpha=~
     SkydiveDriftDelay=~
    ).DirectionLast
   ) -> (Pose:Pose Link Structure) {
    is=title AnimGraphNode_BlendListByEnum
    blendType=Linear
    blendPose [
     {linkId=-1, sourceLinkId=-1}
     {linkId=-1, sourceLinkId=-1}
     {linkId=-1, sourceLinkId=-1}
     {linkId=-1, sourceLinkId=-1}
     {linkId=-1, sourceLinkId=-1}
    ]
    blendTime [0.1,0.1,0.1,0.1,0.1]
   }
   Output Animation Pose(Result=@n4)
  ]
 }
 Lean_Into {
  path=AnimGraph.AnimGraphNode_StateMachine_6.Lean Transition.AnimStateNode_17.Lean_Into
  nodes [
   AnimGraphNode_StateMachine_0 = Lean_Into() {is=title AnimGraphNode_StateMachine}
   Output Animation Pose(Result=@AnimGraphNode_StateMachine_0)
  ]
 }
 Lean_Into {
  entry=Identity
  path=AnimGraph.AnimGraphNode_StateMachine_6.Lean Transition.AnimStateNode_17.Lean_Into.AnimGraphNode_StateMachine_0.Lean_Into
  ids=id
  nodes [
   Forward Into() {is=title AnimStateAliasNode, also {aliases [Back_Into]}}
   Back Into() {is=title AnimStateAliasNode, also {aliases [Forward_Into]}}
   AnimStateNode_1 = Identity()
   AnimStateNode_5 = Forward_Into() {is=title AnimStateNode}
   AnimStateNode_6 = Back_Into() {is=title AnimStateNode}
   AnimStateTransitionNode_1 = AnimStateTransitionNode()
   AnimStateTransitionNode_3 = AnimStateTransitionNode()
  ]
  links [
   {
    dst=Back_Into
    rule=AnimGraph.AnimGraphNode_StateMachine_6.Lean Transition.AnimStateNode_17.Lean_Into.AnimGraphNode_StateMachine_0.Lean_Into.AnimStateTransitionNode_1.Transition
    src=Back Into
    via=AnimStateTransitionNode_1
   }
   {
    dst=Forward_Into
    rule=AnimGraph.AnimGraphNode_StateMachine_6.Lean Transition.AnimStateNode_17.Lean_Into.AnimGraphNode_StateMachine_0.Lean_Into.AnimStateTransitionNode_3.Transition
    src=Forward Into
    via=AnimStateTransitionNode_3
   }
  ]
 }
 Identity {
  path=AnimGraph.AnimGraphNode_StateMachine_6.Lean Transition.AnimStateNode_17.Lean_Into.AnimGraphNode_StateMachine_0.Lean_Into.AnimStateNode_1.Identity
  nodes [
   n1 = "Sequence Player (additive) 'PuzzleShed_Lean_Center_Pose_GLIDER'"() {bLoopAnimation=false, sequence=&PuzzleShed_Lean_Center_Pose_GLIDER}
   Output Animation Pose(Result=@n1)
  ]
 }
 Forward_Into {
  path=AnimGraph.AnimGraphNode_StateMachine_6.Lean Transition.AnimStateNode_17.Lean_Into.AnimGraphNode_StateMachine_0.Lean_Into.AnimStateNode_5.Forward_Into
  nodes [
   n1 = "Sequence Player (additive) 'PuzzleShed_Lean_Forward_Into_Anim_GLIDER'"() {bLoopAnimation=false, sequence=&/PuzzleShed_Lean_Forward_Into_Anim_GLIDER}
   n2 = "Sequence Player (additive) 'PuzzleShed_Lean_Center_Pose_GLIDER'"() {bLoopAnimation=false, sequence=&PuzzleShed_Lean_Center_Pose_GLIDER}
   Output Animation Pose(Result=Apply Additive!(Base=@n2, Additive=@n1, Alpha:float=0.3) -> (Pose) {alphaBoolBlend {blendOption=Linear}})
  ]
 }
 Back_Into {
  path=AnimGraph.AnimGraphNode_StateMachine_6.Lean Transition.AnimStateNode_17.Lean_Into.AnimGraphNode_StateMachine_0.Lean_Into.AnimStateNode_6.Back_Into
  nodes [
   n1 = "Sequence Player (additive) 'PuzzleShed_Lean_Center_Pose_GLIDER'"() {bLoopAnimation=false, sequence=&PuzzleShed_Lean_Center_Pose_GLIDER}
   n2 = "Sequence Player (additive) 'PuzzleShed_Lean_Back_Into_Anim_GLIDER'"() {bLoopAnimation=false, sequence=&/PuzzleShed_Lean_Back_Into_Anim_GLIDER}
   Output Animation Pose(Result=Apply Additive(Base=@n1, Additive=@n2, Alpha=0.3) {alphaBoolBlend {blendOption=Linear}})
  ]
 }
 Transition {
  path=AnimGraph.AnimGraphNode_StateMachine_6.Lean Transition.AnimStateNode_17.Lean_Into.AnimGraphNode_StateMachine_0.Lean_Into.AnimStateTransitionNode_1.Transition
  nodes [Result(bCanEnterTransition=Break Fort Anim Input Skydiving!(FortAnimInput_Skydiving=Get SkydivingState()) -> (
     bIsGliderRight=~
     bIsGliderCenter=~
     bIsGliderLeft=~
     bIsGliderForward=~
     bIsGliderBack=~
     bHasDeployedAndIsStandingUpright=~
     LocalAccelDir
     PoseAdditive
    ).bIsGliderBack)]
 }
 Transition {
  path=AnimGraph.AnimGraphNode_StateMachine_6.Lean Transition.AnimStateNode_17.Lean_Into.AnimGraphNode_StateMachine_0.Lean_Into.AnimStateTransitionNode_3.Transition
  nodes [Result(bCanEnterTransition=Break Fort Anim Input Skydiving(FortAnimInput_Skydiving=Get SkydivingState()) -> (bIsGliderRight=~, bIsGliderCenter=~, bIsGliderLeft=~, bIsGliderForward=~, bIsGliderBack=~, bHasDeployedAndIsStandingUpright=~).bIsGliderForward)]
 }
 AnimationTransitionGraph_1 {path=AnimGraph.AnimGraphNode_StateMachine_6.Lean Transition.AnimStateTransitionNode_26.AnimationTransitionGraph_1, nodes [Result()]}
 AnimationTransitionGraph_1 {
  path=AnimGraph.AnimGraphNode_StateMachine_6.Lean Transition.AnimStateTransitionNode_37.AnimationTransitionGraph_1
  nodes [
   n1 = Get bIsAccelerating2D()
   Result(bCanEnterTransition=@n1)
   n2 = Get Skydiving() -> (Skydiving:Fort Anim Input Skydiving Structure)
   OR Boolean!(
    A="Equal (Enum)"(
     A:ESkydivingDirection Enum=Break Fort Anim Input Skydiving(FortAnimInput_Skydiving=@n2) -> (bIsGliderRight=~, bIsGliderCenter=~, bIsGliderLeft=~, bIsGliderForward=~, bIsGliderBack=~, bHasDeployedAndIsStandingUpright=~).LocalAccelDir
     B:ESkydivingDirection Enum=Center
    )
    B=NOT Boolean(A=@n1)
   ) -> (ReturnValue)
  ]
 }
 AnimationTransitionGraph_1 {
  path=AnimGraph.AnimGraphNode_StateMachine_6.Lean Transition.AnimStateTransitionNode_38.AnimationTransitionGraph_1
  nodes [
   n1 = NOT Boolean(A=Get bIsAccelerating2D())
   Result(bCanEnterTransition=@n1)
   n2 = Break Fort Anim Input Skydiving(FortAnimInput_Skydiving=Get Skydiving()) -> (bIsGliderRight=~, bIsGliderCenter=~, bIsGliderLeft=~, bIsGliderForward=~, bIsGliderBack=~, bHasDeployedAndIsStandingUpright=~)
   OR Boolean(A=@n2.bIsGliderCenter, B=@n1)
   "Equal (Enum)"(A=@n2.LocalAccelDir, B=Center)
  ]
 }
 AnimationTransitionGraph_1 {
  path=AnimGraph.AnimGraphNode_StateMachine_6.Lean Transition.AnimStateTransitionNode_40.AnimationTransitionGraph_1
  nodes [Result(bCanEnterTransition=Get bIsAccelerating2D())]
 }
 Lean Transition_1 {
  entry=Lean_Zero
  path=AnimGraph.AnimGraphNode_StateMachine_8.Lean Transition_1
  ids=id
  nodes [
   AnimStateNode_10 = Lean_Zero() {
    stateEntered {guid=#10}
    stateFullyBlended {guid=#11}
    stateLeft {guid=#12}
   }
   AnimStateNode_14 = Lean_Out() {
    stateEntered {guid=#13}
    stateFullyBlended {guid=#14}
    stateLeft {guid=#15}
   }
   AnimStateNode_17 = Lean_Into() {
    is=title AnimStateNode
    stateEntered {guid=#16}
    stateFullyBlended {guid=#17}
    stateLeft {guid=#18}
   }
   AnimStateTransitionNode_26 = AnimStateTransitionNode() {
    bAutomaticRuleBasedOnSequencePlayerInState=true
    transitionEnd {guid=#19}
    transitionInterrupt {guid=#20}
    transitionStart {guid=#21}
   }
   AnimStateTransitionNode_37 = AnimStateTransitionNode()
   AnimStateTransitionNode_38 = AnimStateTransitionNode() {
    transitionEnd {guid=#22}
    transitionInterrupt {guid=#23}
    transitionStart {guid=#24}
   }
   AnimStateTransitionNode_40 = AnimStateTransitionNode() {
    transitionEnd {guid=#25}
    transitionInterrupt {guid=#26}
    transitionStart {guid=#27}
   }
  ]
  links [
   {
    dst=Lean_Out
    rule=AnimGraph.AnimGraphNode_StateMachine_8.Lean Transition_1.AnimStateTransitionNode_38.AnimationTransitionGraph_1
    src=Lean_Into
    via=AnimStateTransitionNode_38
   }
   {
    dst=Lean_Into
    rule=AnimGraph.AnimGraphNode_StateMachine_8.Lean Transition_1.AnimStateTransitionNode_40.AnimationTransitionGraph_1
    src=Lean_Out
    via=AnimStateTransitionNode_40
   }
   {
    dst=Lean_Zero
    rule=AnimGraph.AnimGraphNode_StateMachine_8.Lean Transition_1.AnimStateTransitionNode_26.AnimationTransitionGraph_1
    src=Lean_Out
    via=AnimStateTransitionNode_26
   }
   {
    dst=Lean_Into
    rule=AnimGraph.AnimGraphNode_StateMachine_8.Lean Transition_1.AnimStateTransitionNode_37.AnimationTransitionGraph_1
    src=Lean_Zero
    via=AnimStateTransitionNode_37
   }
  ]
 }
 Lean_Zero {
  path=AnimGraph.AnimGraphNode_StateMachine_8.Lean Transition_1.AnimStateNode_10.Lean_Zero
  nodes [
   n1 = "Sequence Player (additive) 'PuzzleShed_Lean_Center_Pose_GLIDER'"() {bLoopAnimation=false, sequence=&PuzzleShed_Lean_Center_Pose_GLIDER}
   Output Animation Pose(Result=@n1)
  ]
 }
 Lean_Out {
  path=AnimGraph.AnimGraphNode_StateMachine_8.Lean Transition_1.AnimStateNode_14.Lean_Out
  nodes [
   AnimGraphNode_StateMachine_0 = New State Machine() {is=title AnimGraphNode_StateMachine}
   Output Animation Pose(Result=@AnimGraphNode_StateMachine_0)
  ]
 }
 New State Machine {
  entry=Lean_Zero
  path=AnimGraph.AnimGraphNode_StateMachine_8.Lean Transition_1.AnimStateNode_14.Lean_Out.AnimGraphNode_StateMachine_0.New State Machine
  ids=id
  nodes [
   Forward Into() {also {aliases [Back_Into,Lean_Zero]}}
   Back Into() {also {aliases [Forward_Into,Lean_Zero]}}
   AnimStateNode_0 = Lean_Zero() {
    stateEntered {guid=#28}
    stateFullyBlended {guid=#29}
    stateLeft {guid=#30}
   }
   AnimStateNode_5 = Forward_Into()
   AnimStateNode_6 = Back_Into()
   AnimStateTransitionNode_1 = AnimStateTransitionNode()
   AnimStateTransitionNode_3 = AnimStateTransitionNode()
  ]
  links [
   {
    dst=Back_Into
    rule=AnimGraph.AnimGraphNode_StateMachine_8.Lean Transition_1.AnimStateNode_14.Lean_Out.AnimGraphNode_StateMachine_0.New State Machine.AnimStateTransitionNode_1.AnimationTransitionGraph_0
    src=Back Into
    via=AnimStateTransitionNode_1
   }
   {
    dst=Forward_Into
    rule=AnimGraph.AnimGraphNode_StateMachine_8.Lean Transition_1.AnimStateNode_14.Lean_Out.AnimGraphNode_StateMachine_0.New State Machine.AnimStateTransitionNode_3.AnimationTransitionGraph_0
    src=Forward Into
    via=AnimStateTransitionNode_3
   }
  ]
 }
 Lean_Zero {
  path=AnimGraph.AnimGraphNode_StateMachine_8.Lean Transition_1.AnimStateNode_14.Lean_Out.AnimGraphNode_StateMachine_0.New State Machine.AnimStateNode_0.Lean_Zero
  nodes [
   n1 = "Sequence Player (additive) 'PuzzleShed_Lean_Center_Pose_GLIDER'"() {bLoopAnimation=false, sequence=&PuzzleShed_Lean_Center_Pose_GLIDER}
   Output Animation Pose(Result=@n1)
  ]
 }
 Forward_Into {
  path=AnimGraph.AnimGraphNode_StateMachine_8.Lean Transition_1.AnimStateNode_14.Lean_Out.AnimGraphNode_StateMachine_0.New State Machine.AnimStateNode_5.Forward_Into
  nodes [
   n1 = "Sequence Player (additive) 'PuzzleShed_Lean_Left_Out_Anim_GLIDER'"() {bLoopAnimation=false, sequence=&/PuzzleShed_Lean_Left_Out_Anim_GLIDER}
   Output Animation Pose(Result=@n1)
  ]
 }
 Back_Into {
  path=AnimGraph.AnimGraphNode_StateMachine_8.Lean Transition_1.AnimStateNode_14.Lean_Out.AnimGraphNode_StateMachine_0.New State Machine.AnimStateNode_6.Back_Into
  nodes [
   n1 = "Sequence Player (additive) 'PuzzleShed_Lean_Right_Out_Anim_GLIDER'"() {bLoopAnimation=false, sequence=&/PuzzleShed_Lean_Right_Out_Anim_GLIDER}
   Output Animation Pose(Result=@n1)
  ]
 }
 AnimationTransitionGraph_0 {
  path=AnimGraph.AnimGraphNode_StateMachine_8.Lean Transition_1.AnimStateNode_14.Lean_Out.AnimGraphNode_StateMachine_0.New State Machine.AnimStateTransitionNode_1.AnimationTransitionGraph_0
  nodes [Result(bCanEnterTransition=Get Glide_right())]
 }
 AnimationTransitionGraph_0 {
  path=AnimGraph.AnimGraphNode_StateMachine_8.Lean Transition_1.AnimStateNode_14.Lean_Out.AnimGraphNode_StateMachine_0.New State Machine.AnimStateTransitionNode_3.AnimationTransitionGraph_0
  nodes [Result(bCanEnterTransition=Get Glide_Left())]
 }
 Lean_Into {
  path=AnimGraph.AnimGraphNode_StateMachine_8.Lean Transition_1.AnimStateNode_17.Lean_Into
  nodes [
   AnimGraphNode_StateMachine_0 = Lean_Into() {is=title AnimGraphNode_StateMachine}
   Output Animation Pose(Result=@AnimGraphNode_StateMachine_0)
  ]
 }
 Lean_Into {
  entry=Lean_Zero
  path=AnimGraph.AnimGraphNode_StateMachine_8.Lean Transition_1.AnimStateNode_17.Lean_Into.AnimGraphNode_StateMachine_0.Lean_Into
  ids=id
  nodes [
   Forward Into() {also {aliases [Lean_Zero]}}
   Back Into() {also {aliases [Lean_Zero]}}
   AnimStateNode_0 = Lean_Zero() {
    stateEntered {guid=#31}
    stateFullyBlended {guid=#32}
    stateLeft {guid=#33}
   }
   AnimStateNode_5 = Forward_Into()
   AnimStateNode_6 = Back_Into()
   AnimStateTransitionNode_1 = AnimStateTransitionNode()
   AnimStateTransitionNode_3 = AnimStateTransitionNode()
  ]
  links [
   {
    dst=Back_Into
    rule=AnimGraph.AnimGraphNode_StateMachine_8.Lean Transition_1.AnimStateNode_17.Lean_Into.AnimGraphNode_StateMachine_0.Lean_Into.AnimStateTransitionNode_1.AnimationTransitionGraph_0
    src=Back Into
    via=AnimStateTransitionNode_1
   }
   {
    dst=Forward_Into
    rule=AnimGraph.AnimGraphNode_StateMachine_8.Lean Transition_1.AnimStateNode_17.Lean_Into.AnimGraphNode_StateMachine_0.Lean_Into.AnimStateTransitionNode_3.AnimationTransitionGraph_0
    src=Forward Into
    via=AnimStateTransitionNode_3
   }
  ]
 }
 Lean_Zero {
  path=AnimGraph.AnimGraphNode_StateMachine_8.Lean Transition_1.AnimStateNode_17.Lean_Into.AnimGraphNode_StateMachine_0.Lean_Into.AnimStateNode_0.Lean_Zero
  nodes [
   n1 = "Sequence Player (additive) 'PuzzleShed_Lean_Center_Pose_GLIDER'"() {bLoopAnimation=false, sequence=&PuzzleShed_Lean_Center_Pose_GLIDER}
   Output Animation Pose(Result=@n1)
  ]
 }
 Forward_Into {
  path=AnimGraph.AnimGraphNode_StateMachine_8.Lean Transition_1.AnimStateNode_17.Lean_Into.AnimGraphNode_StateMachine_0.Lean_Into.AnimStateNode_5.Forward_Into
  nodes [
   n1 = "Sequence Player (additive) 'PuzzleShed_Lean_Left_Into_Anim_GLIDER'"!() -> (Pose) {bLoopAnimation=false, sequence=&PuzzleShed_Lean_Left_Into_Anim_GLIDER}
   Output Animation Pose(Result=@n1)
  ]
 }
 Back_Into {
  path=AnimGraph.AnimGraphNode_StateMachine_8.Lean Transition_1.AnimStateNode_17.Lean_Into.AnimGraphNode_StateMachine_0.Lean_Into.AnimStateNode_6.Back_Into
  nodes [
   n1 = "Sequence Player (additive) 'PuzzleShed_Lean_Right_Into_Anim_GLIDER'"!() -> (Pose) {bLoopAnimation=false, sequence=&PuzzleShed_Lean_Right_Into_Anim_GLIDER}
   Output Animation Pose(Result=@n1)
  ]
 }
 AnimationTransitionGraph_0 {
  path=AnimGraph.AnimGraphNode_StateMachine_8.Lean Transition_1.AnimStateNode_17.Lean_Into.AnimGraphNode_StateMachine_0.Lean_Into.AnimStateTransitionNode_1.AnimationTransitionGraph_0
  nodes [Result(bCanEnterTransition=Get Glide_right())]
 }
 AnimationTransitionGraph_0 {
  path=AnimGraph.AnimGraphNode_StateMachine_8.Lean Transition_1.AnimStateNode_17.Lean_Into.AnimGraphNode_StateMachine_0.Lean_Into.AnimStateTransitionNode_3.AnimationTransitionGraph_0
  nodes [Result(bCanEnterTransition=Get Glide_Left())]
 }
 AnimationTransitionGraph_1 {path=AnimGraph.AnimGraphNode_StateMachine_8.Lean Transition_1.AnimStateTransitionNode_26.AnimationTransitionGraph_1, nodes [Result()]}
 AnimationTransitionGraph_1 {
  path=AnimGraph.AnimGraphNode_StateMachine_8.Lean Transition_1.AnimStateTransitionNode_37.AnimationTransitionGraph_1
  nodes [Result(bCanEnterTransition=Get Yaw Input())]
 }
 AnimationTransitionGraph_1 {
  path=AnimGraph.AnimGraphNode_StateMachine_8.Lean Transition_1.AnimStateTransitionNode_38.AnimationTransitionGraph_1
  nodes [Result(bCanEnterTransition=NOT Boolean(A=Get Yaw Input()))]
 }
 AnimationTransitionGraph_1 {
  path=AnimGraph.AnimGraphNode_StateMachine_8.Lean Transition_1.AnimStateTransitionNode_40.AnimationTransitionGraph_1
  nodes [Result(bCanEnterTransition=Get Yaw Input())]
 }
 EventGraph {nodes [
   AnimNotify_EndBlink() {is=title K2Node_Event}
   > n1 = Branch(Condition="Equal (Integer)"(A=RandomInteger!(Max:int=5) -> (ReturnValue:int=0), B=3))
   > Delay(Duration=0.4)
   > n2 = Set IsBlinking(IsBlinking:bool=false) -> (Output_Get:bool)
   Set IsBlinking(execute=[@n28.Completed, @n2.then, @n3.then], IsBlinking=true)
   Delay(execute=@n1.else, Duration=RandomFloatInRange(Min=2.5, Max=10))
   > n3 = Set IsBlinking(IsBlinking=false)
   AnimNotify_TurnONEngines() {is=title K2Node_Event}
   > SetGliderFullyDeployed!(execute, self:Fort Glider Anim Instance Object Reference, bIsFullyDeployed:bool=true) -> (then)
   n4 = Event BlueprintUpdateAnimation()
   > Set Time(Time:double=@n4.DeltaTimeX) -> (Output_Get:double)
   > n5 = Sequence!(execute) -> (then_0, then_1, then_2:Exec, then_3:Exec, then_4:Exec)
   > Get Is Parachute Open!(execute:Exec) {is=title K2Node_MacroInstance}
   n6 = Sequence!(execute=@n5.then_4) -> (then_0, then_1, then_2, then_3, then_4, then_5:Exec)
   n7 = Break Fort Anim Input Skydiving!(FortAnimInput_Skydiving=Get SkydivingState()) -> (
    ParaGilderRootOffset:Vector=~
    ParaGliderRootRotationOverride:Rotator=~
    bUseParaGlideRootModifier:bool=~
    bIsSkydivingFromLaunchPad:bool=~
    bIsSkydivingFromBus:bool=~
    bIsInVortex:bool=~
    bIsDiving:bool=~
    bIsParachuteOpen:bool=~
    bIsSkydiving:bool=~
    bIsParachuteLeaning:bool=~
    bIsSkydiveLeaning:bool=~
    bIsLeaning:bool=~
    bIsSkydiveDiveMode:bool=~
    bPlaySkydiveDrift:bool=~
    bSkydiveDriftDelayActive:bool=~
    bSkydiveDriftAnimAllowed:bool=~
    bIsGliderRight=~
    bIsGliderCenter=~
    bIsGliderLeft=~
    bIsGliderForward=~
    bIsGliderBack=~
    bHasDeployedAndIsStandingUpright=~
    LocalAccelForward:float=~
    LocalAccelRight:float=~
    LocalVelocityRight:float=~
    SkydiveAimPitch:float=~
    SkydiveAimYaw:float=~
    DeployChuteAnimRate:float=~
    SkydiveDriftAnimRate:float=~
    SkydiveDriftAnimRateCurrent:float=~
    SkydiveFidgetAnimRate:float=~
    SkydiveFidgetAnimRateCurrent:float=~
    SkydiveAdditiveAlpha:float=~
    SkydiveDriftDelay:float=~
    SkydiveDriftAnim:int
    SkydiveDriftAnimMax:int
    LaunchpadAnim:int
    LocalAccelDir
    DirectionLast:ESkydivingDirection Enum
    PoseAdditive
   )
   > n8 = Branch(Condition="float < float"(A=@n7.SkydiveAimYaw, B=-0.001))
   > Set Glide_right(Glide_right:bool=false) -> (Output_Get:bool)
   > Set Glide_Left(Glide_Left:bool=true) -> (Output_Get:bool)
   Set Glide_Left(execute=@n8.else, Glide_Left=false)
   n9 = Cast To PlayerPawn_Athena(execute=@n6.then_2, Object=GetPlayerPawn!(PlayerIndex:int=0) -> (ReturnValue:Pawn Object Reference))
   > Delay(Duration=0.1)
   n10 = BreakRotator(InRot=GetControlRotation!(self:Pawn Object Reference=@n9) -> (ReturnValue:Rotator))
   > n11 = Set previous_rotation_value(previous_rotation_value:double=@n10.Pitch) -> (Output_Get:double)
   > Delay(Duration=0.05)
   > n12 = Set new_rotation_value(new_rotation_value:double=@n10.Pitch) -> (Output_Get:double)
   n13 = Branch(execute=@n6.then_1, Condition="float > float"(A=@n7.SkydiveAimYaw, B=0.001))
   > Set Glide_Left(Glide_Left=false)
   > Set Glide_right(Glide_right=true)
   Set Glide_right(execute=@n13.else, Glide_right=false)
   n14 = float / float(
    A=Break Fort Anim Input Skydiving(FortAnimInput_Skydiving=Get SkydivingState()) -> (
     ParaGilderRootOffset=~
     ParaGliderRootRotationOverride=~
     bUseParaGlideRootModifier=~
     bIsSkydivingFromLaunchPad=~
     bIsSkydivingFromBus=~
     bIsInVortex=~
     bIsDiving=~
     bIsParachuteOpen=~
     bIsSkydiving=~
     bIsParachuteLeaning=~
     bIsSkydiveLeaning=~
     bIsLeaning=~
     bIsSkydiveDiveMode=~
     bPlaySkydiveDrift=~
     bSkydiveDriftDelayActive=~
     bSkydiveDriftAnimAllowed=~
     bIsGliderRight=~
     bIsGliderCenter=~
     bIsGliderLeft=~
     bIsGliderForward=~
     bIsGliderBack=~
     bHasDeployedAndIsStandingUpright=~
     LocalAccelForward=~
     LocalAccelRight=~
     LocalVelocityRight=~
     SkydiveAimPitch=~
     SkydiveAimYaw=~
     DeployChuteAnimRate=~
     SkydiveDriftAnimRate=~
     SkydiveDriftAnimRateCurrent=~
     SkydiveFidgetAnimRate=~
     SkydiveFidgetAnimRateCurrent=~
     SkydiveAdditiveAlpha=~
     SkydiveDriftDelay=~
    ).SkydiveAimYaw
    B=45
   )
   n15 = Branch(execute=@n6.then_4, Condition="float > float"(A=@n14))
   > Set Yaw Velocity(Yaw Velocity:double=@n14) -> (Output_Get:double)
   Set Yaw Velocity(execute=@n15.else, Yaw Velocity=float * float(A=@n14, B=-1) {is=title K2Node_PromotableOperator})
   n16 = "Equal (Float)"(A=@n12, B=@n11) {is=title K2Node_PromotableOperator}
   n17 = Branch(execute=@n6.then_3, Condition=@n16)
   > Delay(Duration=0.2)
   > Branch(Condition=@n16)
   > Set Yaw Input(Yaw Input:bool=false) -> (Output_Get:bool)
   Set Yaw Input(execute=@n17.else, Yaw Input=true)
   n18 = GetOrCreateDynamicOverlayMaterial(execute=@n5.then_3, MeshComponent=GetOwningComponent(), MaterialName=MI_Glider_PuzzleShed_InvHullOutline)
   > n19 = Branch(Condition="float > float"(A=GetCurveValue(CurveName=Blink_Curve), B=0.1))
   > SetScalarParameterValue(self=@n18, ParameterName=Hide Element 01, Value=1)
   > SetScalarParameterValue(self=@n18, ParameterName=Hide Element 02, Value=0)
   SetScalarParameterValue(execute=@n19.else, self=@n18, ParameterName=Hide Element 01, Value=0)
   > SetScalarParameterValue(self=@n18, ParameterName=Hide Element 02, Value=1)
   n20 = ForwardBack_Value_Interp!(
    execute=@n5.then_2
    Time:double=Get Time() -> (Time:double)
    Currnet_Value:double=Get Acceleration()
    Idle_Value:double=1
    Idle_InterpSpeed:double=1
    Forward_Value:double=2
    Forward_InterpSpeed:double=2
    Back_Value:double=1
    Back_InterpSpeed:double=122
    LocalAccelForward:double=Break Fort Anim Input Skydiving(FortAnimInput_Skydiving=Get SkydivingState()) -> (
     ParaGilderRootOffset=~
     ParaGliderRootRotationOverride=~
     bUseParaGlideRootModifier=~
     bIsSkydivingFromLaunchPad=~
     bIsSkydivingFromBus=~
     bIsInVortex=~
     bIsDiving=~
     bIsParachuteOpen=~
     bIsSkydiving=~
     bIsParachuteLeaning=~
     bIsSkydiveLeaning=~
     bIsLeaning=~
     bIsSkydiveDiveMode=~
     bPlaySkydiveDrift=~
     bSkydiveDriftDelayActive=~
     bSkydiveDriftAnimAllowed=~
     bIsGliderRight=~
     bIsGliderCenter=~
     bIsGliderLeft=~
     bIsGliderForward=~
     bIsGliderBack=~
     bHasDeployedAndIsStandingUpright=~
     LocalAccelForward=~
     LocalAccelRight=~
     LocalVelocityRight=~
     SkydiveAimPitch=~
     SkydiveAimYaw=~
     DeployChuteAnimRate=~
     SkydiveDriftAnimRate=~
     SkydiveDriftAnimRateCurrent=~
     SkydiveFidgetAnimRate=~
     SkydiveFidgetAnimRateCurrent=~
     SkydiveAdditiveAlpha=~
     SkydiveDriftDelay=~
    ).LocalAccelForward
   ) -> (then, New_Value:double)
   > Set Acceleration(Acceleration:double=@n20) -> (Output_Get:double)
   n21 = Get Dynamic Material Instance(execute=@n5.then_1, SkeletalMeshComponent=GetOwningComponent())
   > n22 = Branch(Condition="float > float"(A=GetCurveValue(CurveName=Blink_Curve), B=0.1))
   > SetScalarParameterValue(self=@n21, ParameterName=Hide Element 01, Value=1)
   > SetScalarParameterValue(self=@n21, ParameterName=Hide Element 02, Value=0)
   SetScalarParameterValue(execute=@n22.else, self=@n21, ParameterName=Hide Element 01, Value=0)
   > SetScalarParameterValue(self=@n21, ParameterName=Hide Element 02, Value=1)
   n23 = AnimNotify_ClosedEntered() {is=title K2Node_Event}
   n24 = GetOwningComponent()
   HideBoneByName!(
    execute=[@n23.then, @n25.then_1]
    self:Skinned Mesh Component Object Reference=@n24
    BoneName:Name=Glider_Base
    PhysBodyOption:EPhysBodyOp Enum=PBO_None
   ) -> (then)
   AnimNotify_DeployEntered() {is=title K2Node_Event}
   > Delay(Duration=0.1)
   > UnHideBoneByName!(execute, self:Skinned Mesh Component Object Reference=@n24, BoneName:Name=Glider_Base) -> (then)
   Event BlueprintInitializeAnimation()
   > n25 = Sequence!(execute) -> (then_0, then_1)
   > n26 = Branch(Condition=IsCosmeticObjectDisplayedInFrontendContext!(CosmeticObject:Object Reference) -> (ReturnValue:bool))
   > Set IsInFrontEnd(IsInFrontEnd:bool=true) -> (Output_Get:bool)
   Set IsInFrontEnd(execute=@n26.else, IsInFrontEnd=false)
   Event BlueprintPostEvaluateAnimation() {is=title K2Node_Event}
   > Set IsBlinking(IsBlinking=false)
   AnimNotify_StartBlink() {is=title K2Node_Event}
   > Delay(Duration=RandomFloatInRange(Min=0.7, Max=3))
   > n27 = Branch()
   > n28 = Do Once(Reset=@n27.else)
   Set Time(Time=0)
   > n29 = ForwardBack_Value_Interp(
    Time=Get Time()
    Currnet_Value=Get Example() -> (Example:double)
    Idle_Value=1
    Idle_InterpSpeed=1
    Forward_Value=2
    Forward_InterpSpeed=2
    Back_Value=1
    Back_InterpSpeed=2
    LocalAccelForward=Break Fort Anim Input Skydiving(FortAnimInput_Skydiving=Get SkydivingState()) -> (
     ParaGilderRootOffset=~
     ParaGliderRootRotationOverride=~
     bUseParaGlideRootModifier=~
     bIsSkydivingFromLaunchPad=~
     bIsSkydivingFromBus=~
     bIsInVortex=~
     bIsDiving=~
     bIsParachuteOpen=~
     bIsSkydiving=~
     bIsParachuteLeaning=~
     bIsSkydiveLeaning=~
     bIsLeaning=~
     bIsSkydiveDiveMode=~
     bPlaySkydiveDrift=~
     bSkydiveDriftDelayActive=~
     bSkydiveDriftAnimAllowed=~
     bIsGliderRight=~
     bIsGliderCenter=~
     bIsGliderLeft=~
     bIsGliderForward=~
     bIsGliderBack=~
     bHasDeployedAndIsStandingUpright=~
     LocalAccelForward=~
     LocalAccelRight=~
     LocalVelocityRight=~
     SkydiveAimPitch=~
     SkydiveAimYaw=~
     DeployChuteAnimRate=~
     SkydiveDriftAnimRate=~
     SkydiveDriftAnimRateCurrent=~
     SkydiveFidgetAnimRate=~
     SkydiveFidgetAnimRateCurrent=~
     SkydiveAdditiveAlpha=~
     SkydiveDriftDelay=~
    ).LocalAccelForward
   )
   > Set Example(Example:double=@n29) -> (Output_Get:double)
   n30 = Get OwnerGlider()
   Set Actor Hidden In Game!(execute, self:Actor Object Reference=@n30, bNewHidden:bool=true) -> (then) {is=title K2Node_CallFunction}
   Set Actor Hidden In Game(self=@n30, bNewHidden=true)
   Set Use Alt Glider Cam!(execute:Exec) -> (then:Exec) {is=title K2Node_MacroInstance}
   Get Player Open Glider Bool!(execute:Exec) -> (then:Exec) {is=title K2Node_MacroInstance}
  ]}
 GetIsParachuteOpen {nodes [
   Inputs()
   > n1 = Branch(Condition=Break Fort Anim Input Skydiving(FortAnimInput_Skydiving=Get SkydivingState()) -> (
     ParaGilderRootOffset=~
     ParaGliderRootRotationOverride=~
     bUseParaGlideRootModifier=~
     bIsSkydivingFromLaunchPad=~
     bIsSkydivingFromBus=~
     bIsInVortex=~
     bIsDiving=~
     bIsParachuteOpen=~
     bIsSkydiving=~
     bIsParachuteLeaning=~
     bIsSkydiveLeaning=~
     bIsLeaning=~
     bIsSkydiveDiveMode=~
     bPlaySkydiveDrift=~
     bSkydiveDriftDelayActive=~
     bSkydiveDriftAnimAllowed=~
     bIsGliderRight=~
     bIsGliderCenter=~
     bIsGliderLeft=~
     bIsGliderForward=~
     bIsGliderBack=~
     bHasDeployedAndIsStandingUpright=~
     LocalAccelForward=~
     LocalAccelRight=~
     LocalVelocityRight=~
     SkydiveAimPitch=~
     SkydiveAimYaw=~
     DeployChuteAnimRate=~
     SkydiveDriftAnimRate=~
     SkydiveDriftAnimRateCurrent=~
     SkydiveFidgetAnimRate=~
     SkydiveFidgetAnimRateCurrent=~
     SkydiveAdditiveAlpha=~
     SkydiveDriftDelay=~
    ).bIsParachuteOpen)
   > Set IsParachuteOpen(IsParachuteOpen:bool=true) -> (Output_Get:bool)
   Set IsParachuteOpen(execute=@n1.else, IsParachuteOpen=false)
   Outputs()
  ]}
 GetPlayerOpenGliderBool {nodes [
   Inputs()
   > n1 = Branch(Condition=Get IsCosmeticPreview(self=Get OwnerGlider()))
   Outputs!(then:Exec=[@n3.then, @n1.then])
   n2 = Get OwnerFortPawn() -> (OwnerFortPawn:Fort Player Pawn Object Reference)
   n3 = Set bForceOpen(
    execute=@n1.else
    bForceOpen:bool=ShouldForceOpenGlider!(self:Fort Player Animation Interface Object Reference=GetAnimationInterface!(self:Fort Player Pawn Object Reference=@n2) -> (ReturnValue:Fort Player Animation Interface Object Reference)) -> (ReturnValue:bool)
   ) -> (Output_Get:bool)
  ]}
 SetUseAltGliderCam {nodes [
   Inputs()
   > n1 = Branch(Condition=Get IsCosmeticPreview(self=Get OwnerGlider()))
   Outputs(then=[@n2.then, @n1.then])
   n2 = SetUseAltGliderCamera!(
    execute=@n1.else
    self:Fort Player Animation Interface Object Reference=GetAnimationInterface(self=Get OwnerFortPawn())
    bUseAltGliderCamera:bool=true
   ) -> (then)
  ]}
]
variables [
 Acceleration {type=number}
 AimPitch {type=number}
 DeployAnimRate {type=number}
 Example {type=number}
 Forward {type=number}
 Glide_Left {type=boolean}
 Glide_right {type=boolean}
 IsBlinking {type=boolean}
 IsInFrontEnd {type=boolean}
 IsParachuteOpen {type=boolean}
 new_rotation_value {type=number}
 previous_rotation_value {type=number}
 Right {type=number}
 Time {type=number}
 UseDynamicsAlpha {type=number}
 Yaw Input {type=boolean}
 Yaw Velocity {type=number}
]
